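# Item animation layer Blueprint for Sprite consumables on the Fortnite player skeleton, overriding parts of the character's pose.
# Jump, fall, land and swim-jump layers play large-consumable additive animations.
# The final pose feeds a cached-pose state machine: Base adjusts hand and jump-additive curves, while Equip layers a sync-group follower animation from the right clavicle.
# Targeting-based aim-offset blend is unwired, most other layers are empty, and the event graph is empty.
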AnimBlueprint /Sprites/Assets/Animation/Sprite_Consumeable_AnimBP
target_skeleton=&Fortnite_M_Avg_Player_Skeleton
graphs [
 AnimGraph {nodes [Output Pose() {layerGroup=None}]}
 EventGraph {links [], nodes []}
 ItemAimOffset {nodes [
   n1 = Input Pose!() -> (Pose, AimOffsetAlpha:double, Yaw:double, Pitch:double) {name=InPose_AimOffset}
   Blend Poses by bool(
    BlendPose_0=AimOffset Player 'WhisperRemedySweetSprite_AO'!(X:float=@n1.Yaw, Y:float=@n1.Pitch, BasePose:Pose Link Structure=Use cached pose 'AimOffsetCache'(), Alpha:float=1) -> (Pose:Pose Link Structure) {blendSpace=/Sprites/Assets/Animation/WhisperRemedySweet/WhisperRemedySweetSprite_AO, alphaBoolBlend {blendOption=Linear}}
    BlendPose_1=AimOffset Player 'M_NoWep_Relaxed_AO'!(X:float=@n1.Yaw, Y:float=@n1.Pitch, BasePose:Pose Link Structure=Use cached pose 'AimOffsetCache'(), Alpha:float=1) -> (Pose:Pose Link Structure) {blendSpace=/Game/Animation/Game/MainPlayer/Poses/Medium/Male/AO/M_NoWep_Relaxed_AO, alphaBoolBlend {blendOption=Linear}}
    bActiveValue=Get bIsTargeting()
   ) {blendType=Linear}
   Input Pose!() -> (Pose) {name=InPose_PreAimOffsets}
   Output Pose() {layerGroup=Item Override}
   Save cached pose 'AimOffsetCache'() {cacheName=AimOffsetCache}
  ]}
 ItemFallAdditive {nodes [
   Input Pose() {name=InPoseItemFallAdditive}
   Output Pose(Result="Sequence Player (additive) 'ConsumableLarge_JumpFall_Additive_CMM'"!(PlayRate:float=Get FallPlayRate() -> (FallPlayRate:float)) -> (Pose) {sequence=/Game/Animation/Game/MainPlayer/Locomotion/Medium/Male/Jump/ConsumableLarge/ConsumableLarge_JumpFall_Additive_CMM}) {layerGroup=Item Override}
  ]}
 ItemFallLandAdditive {nodes [
   Input Pose() {name=InPoseItemFallLandAdditive}
   n1 = "Sequence Player (additive) 'ConsumableLarge_JumpLand_Additive_CMM'"() {
    bLoopAnimation=false
    groupName=JumpLand
    method=SyncGroup
    sequence=/Game/Animation/Game/MainPlayer/Locomotion/Medium/Male/Jump/ConsumableLarge/ConsumableLarge_JumpLand_Additive_CMM
   }
   Output Pose(Result=@n1) {layerGroup=Item Override}
  ]}
 ItemFallLevel2Override {nodes [
   Input Pose() {name=InPose_FallLevel2}
   Output Pose() {layerGroup=Item Override}
  ]}
 ItemFinalPoseOverride {nodes [
   AnimGraphNode_StateMachine_0 = SpriteOverride() {is=title AnimGraphNode_StateMachine, bAllowConduitEntryStates=true}
   Output Pose(Result=@AnimGraphNode_StateMachine_0) {layerGroup=Item Override}
   Save cached pose 'FinalPoseCache'(Pose=Input Pose() {name=InFinalPose}) {cacheName=FinalPoseCache}
  ]}
 SpriteOverride {
  entry=Base
  path=ItemFinalPoseOverride.AnimGraphNode_StateMachine_0.SpriteOverride
  ids=id
  nodes [
   AnimStateNode_1 = Equip() {is=title AnimStateNode}
   AnimStateNode_4 = Base() {is=title AnimStateNode}
   AnimStateTransitionNode_5 = AnimStateTransitionNode() {crossfadeDuration=0}
   AnimStateTransitionNode_8 = AnimStateTransitionNode()
  ]
  links [
   {
    dst=Equip
    rule=ItemFinalPoseOverride.AnimGraphNode_StateMachine_0.SpriteOverride.AnimStateTransitionNode_5.Transition
    src=Base
    via=AnimStateTransitionNode_5
   }
   {
    dst=Base
    rule=ItemFinalPoseOverride.AnimGraphNode_StateMachine_0.SpriteOverride.AnimStateTransitionNode_8.Transition
    src=Equip
    via=AnimStateTransitionNode_8
   }
  ]
 }
 Equip {
  path=ItemFinalPoseOverride.AnimGraphNode_StateMachine_0.SpriteOverride.AnimStateNode_1.Equip
  nodes [
   n1 = Sequence Player 'Sprite_Equip_CMM'() {bLoopAnimation=false, groupName=EquipSync, groupRole=AlwaysFollower, method=SyncGroup, sequence=&/Sprite_Equip_CMM}
   Output Animation Pose(Result=Layered blend per bone(BasePose=Use cached pose 'FinalPoseCache'(), BlendPoses_0=@n1) {layerSetup [{branchFilters [{boneName=clavicle_r}]}]})
  ]
 }
 Base {
  path=ItemFinalPoseOverride.AnimGraphNode_StateMachine_0.SpriteOverride.AnimStateNode_4.Base
  nodes [
   n1 = Use cached pose 'FinalPoseCache'()
   n2 = Use cached pose 'FinalPoseCache'()
   Output Animation Pose(Result=Layered blend per bone(
     BasePose=Modify Curve(SourcePose=@n2, CurveValues_0=1, Alpha=1) {curveNames [bothhandfk], curveValues [0]}
     BlendPoses_0=Modify Curve!(SourcePose=@n1, CurveValues_0=-2, CurveValues_1:float=1, CurveValues_2:float=-5, Alpha=1) -> (Pose) {curveNames [LeftHandFK, disablejumpupperbodyadditive, bothhandfk], curveValues [0,0,0]}
     BlendWeights_0=Get BlendWeights() -> (BlendWeights:float)
    ))
  ]
 }
 Transition {
  path=ItemFinalPoseOverride.AnimGraphNode_StateMachine_0.SpriteOverride.AnimStateTransitionNode_5.Transition
  nodes [Result(bCanEnterTransition=Get bEquipTransition() -> (bEquipTransition:bool))]
 }
 Transition {
  path=ItemFinalPoseOverride.AnimGraphNode_StateMachine_0.SpriteOverride.AnimStateTransitionNode_8.Transition
  nodes [Result(bCanEnterTransition=Get bExitEquipTransition() -> (bExitEquipTransition:bool))]
 }
 ItemFullBodyAimOffsets {nodes [
   Input Pose!() -> (Pose, FullBody_AimOffsetAlpha:double, Aim_Yaw:double, Aim_Pitch:double) {name=InPoseFullBodyAimOffsets}
   Input Pose!() -> (Pose) {name=InPosePreFullBodyAimOffsets}
   Output Pose() {layerGroup=Item Override}
  ]}
 ItemFullBodyOverride {nodes [
   Input Pose() {name=InPoseFullBody}
   Output Pose() {layerGroup=Item Override}
  ]}
 ItemIdleOverride {nodes [
   Input Pose() {name=InPose_Idle}
   Output Pose() {layerGroup=Item Override}
  ]}
 ItemJetPackJumpAdditive {nodes [Output Pose() {layerGroup=Item Override}]}
 ItemJetPackStartAdditive {nodes [Output Pose() {layerGroup=Item Override}]}
 ItemJumpLoopAdditive {nodes [
   Input Pose() {name=InPoseItemJumpLoopAdditive}
   Output Pose(Result="Sequence Player (additive) 'ConsumableLarge_JumpApex_Additive_CMM'"() {sequence=/Game/Animation/Game/MainPlayer/Locomotion/Medium/Male/Jump/ConsumableLarge/ConsumableLarge_JumpApex_Additive_CMM}) {layerGroup=Item Override}
  ]}
 ItemJumpOverride {nodes [
   Input Pose() {name=InPoseJump}
   Output Pose() {layerGroup=Item Override}
  ]}
 ItemJumpUpAdditive {nodes [
   Input Pose() {name=InPoseItemJumpUpAdditive}
   n1 = "Sequence Player (additive) 'ConsumableLarge_JumpUp_Additive_CMM'"() {
    bLoopAnimation=false
    sequence=/Game/Animation/Game/MainPlayer/Locomotion/Medium/Male/Jump/ConsumableLarge/ConsumableLarge_JumpUp_Additive_CMM
   }
   Output Pose(Result=@n1) {layerGroup=Item Override}
  ]}
 ItemPreAimOffsets {nodes [
   Input Pose() {name=InPosePreAimOffsets}
   Output Pose() {layerGroup=Item Override}
  ]}
 ItemPreIK {nodes [
   Input Pose() {name=InPose_PreIK}
   Output Pose() {layerGroup=Item Override}
  ]}
 ItemRightHandIK {nodes [
   Input Pose() {name=InPosePreRightHandIK}
   Input Pose() {name=InPosePostRightHandIK}
   Output Pose() {layerGroup=Item Override}
  ]}
 ItemSkeletalControl {nodes [
   Input Pose()
   Output Pose() {layerGroup=Item Override}
  ]}
 ItemSwimJumpFallLoop {nodes [
   n1 = "Sequence Player (additive) 'ConsumableLarge_Swim_Jump_Fall_Loop'"() {
    bLoopAnimation=false
    playRate=2
    sequence=/Game/Animation/Game/MainPlayer/Swimming/Jump/ConsumableLarge_Swim_Jump_Fall_Loop
    startPosition=0.25
   }
   Output Pose(Result=@n1) {layerGroup=Item Override}
  ]}
 ItemSwimJumpSurfaceEnd {nodes [Output Pose(Result="Sequence Player (additive) 'ConsumableLarge_Swim_Jump_Surface_End'"() {bLoopAnimation=false, sequence=/Game/Animation/Game/MainPlayer/Swimming/Jump/ConsumableLarge_Swim_Jump_Surface_End}) {layerGroup=Item Override}]}
 ItemSwimJumpSurfaceLoop {nodes [
   n1 = "Sequence Player (additive) 'ConsumableLarge_Swim_Jump_Surface_Loop'"() {playRate=0.5, sequence=/Game/Animation/Game/MainPlayer/Swimming/Jump/ConsumableLarge_Swim_Jump_Surface_Loop, startPosition=0.167}
   Output Pose(Result=@n1) {layerGroup=Item Override}
  ]}
 ItemUpperBody {nodes [
   Output Pose() {layerGroup=Item Override}
   AnimGraphNode_SubInput_0 = Input Pose!() -> (Pose, InputParam:Fort Anim Input Adjusted Aim Structure) {name=InPoseUpperBody}
  ]}
 ItemZipLineStartAdditive {nodes [Output Pose() {layerGroup=Item Override}]}
]
variables []
