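; Actor component for the Irwin simple sprite NPC that drives joyful and fleeing animation states through ISpriteAnimationBlueprint messages.
; A randomized timer switches joyful on for a fixed delay, switches it off, then re-arms itself.
; Each re-arm also starts new Startled gameplay tag listeners on the owner, which set or clear fleeing.
; A helper resolves and caches the anim instance from the owner's mesh, captured at BeginPlay.
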(event EventBeginPlay
  (bind _asnpc_pawn_irwin_simple_sprite (Utilities|Casting|CastToNPC_Pawn_Irwin_Simple_Sprite (Components|GetOwner)))
  (Variables|Default|SetOwnerSpriteMesh (Variables|/Script/Engine.Character|GetMesh _asnpc_pawn_irwin_simple_sprite))
  (CallFunction|RetriggerTimer))

(event Custom|RetriggerTimer
  (bind _returnvalue (Utilities|Time|SetTimerbyEvent (EventDispatchers|CreateEvent :!event Custom|JoyfulWalking) (Math|Random|RandomFloatinRange 2.0 1e+01)))
  (Variables|Default|SetJoyfulTimer _returnvalue)
  (Ability|Tasks|WaitGameplayTagAddToActor (Components|GetOwner) "(TagName=\"Gameplay.Status.Startled\")"
    (:then
      (Ability|Tasks|WaitGameplayTagRemoveFromActor (Components|GetOwner) "(TagName=\"Gameplay.Status.Startled\")"
        (:Removed
          (Class|ISpriteAnimationBlueprint|SetFleeing(Message) (CallFunction|GetAnimInstance) false))))
    (:Added
      (Class|ISpriteAnimationBlueprint|SetFleeing(Message) (CallFunction|GetAnimInstance) true))))

(event Custom|JoyfulWalking
  (Utilities|IsValid (CallFunction|GetAnimInstance)
    (:Is%20Valid
      (Class|ISpriteAnimationBlueprint|SetJoyful(Message) (CallFunction|GetAnimInstance) true)
      (Utilities|FlowControl|Delay 7.0)
      (Class|ISpriteAnimationBlueprint|SetJoyful(Message) (CallFunction|GetAnimInstance) false)
      (Utilities|FlowControl|Delay (Math|Random|RandomFloatinRange 0.2 0.5))
      (Utilities|Time|ClearandInvalidateTimerbyHandle (Variables|Default|GetJoyfulTimer))
      (CallFunction|RetriggerTimer))))

(fn GetAnimInstance
  (bind _ownerspritemesh (Variables|Default|GetOwnerSpriteMesh))
  (bind _returnvalue (Components|SkeletalMesh|GetAnimInstance _ownerspritemesh))
  (Utilities|IsValid _ownerspritemesh
    (:Is%20Valid
      (Utilities|IsValid _returnvalue
        (:Is%20Valid
          (if (not (== _returnvalue (Variables|Default|GetCachedAnimInstance)))
            (Variables|Default|SetCachedAnimInstance _returnvalue))
          (return (Variables|Default|GetCachedAnimInstance)))))))
