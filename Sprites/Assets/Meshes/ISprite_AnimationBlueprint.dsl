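; Blueprint interface for sprite animation Blueprints, exposing setters for a sprite's fleeing and joyful states.
; It declares two functions, SetFleeing and SetJoyful, each taking a single Boolean and returning nothing.
; The file holds only these signatures, leaving the response to each state to whatever implements the interface.
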(fn SetFleeing
  (params (NewValue :type Boolean)))

(fn SetJoyful
  (params (NewValue :type Boolean)))
