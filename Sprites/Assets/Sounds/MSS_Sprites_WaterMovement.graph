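# Mono MetaSound source that plays a looping water-movement sound for Sprites, with loudness driven by movement speed.
# The Fort.Movement speed input is range-mapped to a gain with a faint non-zero floor, smoothed by InterpTo and scaled down.
# That gain multiplies the looping wave player's output.
# The source uses its own attenuation asset, the creature foley sound class and a small concurrency limit.
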MetaSoundSource /Sprites/Assets/Sounds/MSS_Sprites_WaterMovement
inputs [
 Movement.Speed {type=Float}
 Movement.SpeedAzimuth {type=Float}
 Movement.SpeedElevation {type=Float}
 UE.Source.OnPlay {type=Trigger}
 Wave {type=WaveAsset}
]
interfaces [Fort.Movement:1.0, UE.OutputFormat.Mono:1.0, UE.Source:1.1]
outputs ["UE.OutputFormat.Mono.Audio:0" {type=Audio}]
pages [Default {
  input_defaults {
   Movement.Speed=0
   Movement.SpeedAzimuth=0
   Movement.SpeedElevation=0
   UE.Source.OnPlay=null
   Wave=/Sprites/Assets/Sounds/Waves/Sprite_WaterMovement_Loop
  }
  nodes [
   n1 = MapRange.MapRange:Float(In=^Movement.Speed) {also {defaults {In Range B=300, Out Range A=0.07}}}
   WavePlayer_0 = UE.Wave Player:Mono(Play=^UE.Source.OnPlay, Wave Asset=^Wave) {also {defaults {Loop="<MetaSoundBuilderResult.SUCCEEDED: 0>"}}}
   Multiply_0 = UE.Multiply:Audio by Float(
    AdditionalOperands=UE.Multiply:Float(PrimaryOperand=UE.InterpTo:Audio(Target=@n1.Out Value) {also {defaults {Interp Time=0.5}}}.Value) {also {defaults {AdditionalOperands=0.7}}}.Out
    PrimaryOperand=@WavePlayer_0.Out Mono
   )
  ]
  links [{dst="@UE.OutputFormat.Mono.Audio:0", src=Multiply_0.Out}]
 }]
properties {
 assetImportData=&self:AssetImportData
 attenuationSettings=&/AS_Sprites_WaterMovement
 bLooping=true
 bOverrideConcurrency=true
 duration=10000
 maxDistance=2800
 sampleRate=48000
 soundClassObject=/Game/Sounds/FortSoundClassesAndModes/Fort_Creatures_Foley
 volume=0.65
 concurrencyOverrides {maxCount=3}
}
