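# Sound cue for the Sprites feature, named as the feedback played when an action fails because it is on cooldown.
# A single Wave Player feeds the output directly, with no wave asset listed, and routes to the Fort gadget items sound class.
# A concurrency override allows one instance per owner and blocks new plays while one is active, so repeated attempts do not stack.
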SoundCue /Sprites/Assets/Sounds/SC_Sprites_Failed_OnCooldown
root=WavePlayer_0
nodes [{id=WavePlayer_0}]
properties {
 bOverrideConcurrency=true
 maxDistance=2097152
 soundClassObject=&Fort_Gadgets_Items
 concurrencyOverrides {bLimitToOwner=true, maxCount=1, resolutionRule=PreventNew}
}
