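# Sound cue for the Sprites radar ping, playing a randomly chosen ping wave with its own attenuation asset and the Fort_Loot_Collect sound class.
# A modulator varies pitch slightly upward and feeds a short fixed delay ahead of a quality-level switch.
# The first two quality slots draw equally from eight waves, while the last two use only four of them.
# A concurrency override caps simultaneous pings, stopping the farthest first.
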SoundCue /Sprites/Assets/Sounds/SC_Sprites_Radar_Ping
root=Modulator_0
nodes [
 {id=Delay_0, children [QualityLevel_0], properties {delayMax=0.3, delayMin=0.3}}
 {id=Modulator_0, children [Delay_0], properties {pitchMax=1.2, pitchMin=1, volumeMin=1}}
 {id=QualityLevel_0, children [Random_1,Random_1,Random_0,Random_0]}
 {
  id=Random_0
  children [WavePlayer_0,WavePlayer_1,WavePlayer_2,WavePlayer_3]
  properties {bShouldExcludeFromBranchCulling=true, weights [1,1,1,1]}
 }
 {
  id=Random_1
  children [WavePlayer_4, WavePlayer_5, WavePlayer_6, WavePlayer_7, WavePlayer_8, WavePlayer_9, WavePlayer_10, WavePlayer_11]
  properties {bShouldExcludeFromBranchCulling=true, weights [1,1,1,1,1,1,1,1]}
 }
 {id=WavePlayer_0, properties {soundWaveAssetPtr=/Sprites/Assets/Sounds/Waves/Sprite_Radar_Ping_04}}
 {id=WavePlayer_1, properties {soundWaveAssetPtr=/Sprites/Assets/Sounds/Waves/Sprite_Radar_Ping_03}}
 {id=WavePlayer_10, properties {soundWaveAssetPtr=/Sprites/Assets/Sounds/Waves/Sprite_Radar_Ping_06}}
 {id=WavePlayer_11, properties {soundWaveAssetPtr=/Sprites/Assets/Sounds/Waves/Sprite_Radar_Ping_05}}
 {id=WavePlayer_2, properties {soundWaveAssetPtr=/Sprites/Assets/Sounds/Waves/Sprite_Radar_Ping_02}}
 {id=WavePlayer_3, properties {soundWaveAssetPtr=/Sprites/Assets/Sounds/Waves/Sprite_Radar_Ping_01}}
 {id=WavePlayer_4, properties {soundWaveAssetPtr=&Sprite_Radar_Ping_04}}
 {id=WavePlayer_5, properties {soundWaveAssetPtr=&Sprite_Radar_Ping_03}}
 {id=WavePlayer_6, properties {soundWaveAssetPtr=&Sprite_Radar_Ping_02}}
 {id=WavePlayer_7, properties {soundWaveAssetPtr=&Sprite_Radar_Ping_01}}
 {id=WavePlayer_8, properties {soundWaveAssetPtr=/Sprites/Assets/Sounds/Waves/Sprite_Radar_Ping_08}}
 {id=WavePlayer_9, properties {soundWaveAssetPtr=/Sprites/Assets/Sounds/Waves/Sprite_Radar_Ping_07}}
]
properties {
 attenuationSettings=&/AS_Sprites_Radar_Ping
 bOverrideConcurrency=true
 maxDistance=2097152
 soundClassObject=&Fort_Loot_Collect
 volumeMultiplier=1
 concurrencyOverrides {maxCount=5, resolutionRule=StopFarthestThenPreventNew}
}
