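# Material function that outputs a scalar dither pattern for gradient thresholding, combining temporal and Bayer patterns.
# Pixel position, offset by the temporal sample index, feeds a mod-5 diagonal grid and a looped Bayer-style Custom node with a clamped Iterations input trading quality for cost.
# The loop tests bits with logical `&&` rather than a bitwise mask.
# Both patterns are summed to address banding, then scaled down and saturated.
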MaterialFunction /Sprites/Assets/Textures/MF_Dither_TemporalBayer
graph {
 properties {
  description=Generates dither combining temporal and bayer patterns.
  userExposedCaption=Dither Temporal Bayer
  libraryCategoriesText [Global Function Library]
 }
 nodes [
  n1 = ViewProperty() {property=MEVP_TemporalSampleIndex}
  MaterialExpressionAdd_23 = Add(A=ScreenPosition().PixelPosition, B=AppendVector(A=@n1.Property, B=@n1.Property))
  n2 = Custom!(None=@MaterialExpressionAdd_23) {
   code="Mod(((uint)(Input.x) + 2 * (uint)(Input.y)), 5);"
   description=Grid
   outputType=CMOT_Float1
   showCode=true
   inputs [{inputName=Input, input {expression=&self:MaterialExpressionAdd_23}}]
  }
  n3 = FunctionInput(Preview=Constant() {r=2}) {
   bUsePreviewValueAsDefault=true
   description="2 is default, 1 is cheap. 4 is highest quality, but expensive."
   inputName=Iterations
   inputType=FunctionInput_Scalar
   sortPriority=1
   previewValue=<2 0 0 1>
  }
  MaterialExpressionMax_0 = Max(A=@n3)
  n4 = Custom!(UV=@MaterialExpressionAdd_23, Iterations=@MaterialExpressionMax_0) {
   code="// Bayer matrix\r\n\r\nint2 p = int2(UV);\r\n\r\nint sum = 0;\r\n\r\nfor(int i = 0; i < int(Iterations); i++)\r\n{\r\n    int2 t = p && 1;\r\n    sum = sum * 4 | (t.x ^ t.y) * 2 | t.x;\r\n    p *= 0.5;\r\n}    \r\n\r\nreturn float(sum) / float(1 << (2 * int(Iterations)));"
   description=Bayer
   outputType=CMOT_Float1
   showCode=true
   inputs [
    {inputName=UV, input {expression=&self:MaterialExpressionAdd_23}}
    {inputName=Iterations, input {expression=&self:MaterialExpressionMax_0}}
   ]
  }
  n5 = Add(A=@n2, B=@n4) {desc=address banding issues}
  FunctionOutput(None=Saturate(None=Divide(A=@n5) {constB=6})) {description="Dithering pattern. Add to your gradient, subtract 0.5, round, and saturate.", outputName=Pattern}
 ]
}
